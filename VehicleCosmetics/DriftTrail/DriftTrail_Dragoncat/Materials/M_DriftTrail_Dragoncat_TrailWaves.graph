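# Unlit, two-sided translucent Niagara ribbon and mesh-particle material for the Dragoncat vehicle drift trail.
# Color pans a ground-trail texture with optional scrolling glow, then adjusts brightness, contrast and saturation for emissive.
# A separate mask blends random squares with inverted, noise-distorted waves times panning dots, switchable between mirrored and single waves.
# That mask is cut by fine Frac-based stripes along U and scaled by particle alpha.
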Material /VehicleCosmetics/DriftTrail/DriftTrail_Dragoncat/Materials/M_DriftTrail_Dragoncat_TrailWaves
graph {
 properties {
  bAutomaticallySetUsageInEditor=false
  bUsedWithNiagaraMeshParticles=true
  bUsedWithNiagaraRibbons=true
  blendMode=BLEND_Translucent
  shadingModel=MSM_Unlit
  twoSided=true
  usedShadingModels=Unlit
 }
 nodes [
  n1 = Add(
   A=ComponentMask(None=TextureCoordinate()) {r=true}
   B=Multiply(A=Time(), B=ScalarParameter() {defaultValue=-0.7, parameterName=GlowPanningSpd}) {constB=-0.7}
  )
  n2 = Add(A=Power(Base=Frac(None=@n1)) {constExponent=8}) {constB=0.05}
  n3 = Add(A=Multiply(A=Saturate(None=@n2), B=ScalarParameter() {defaultValue=4, parameterName=GlowBrightness})) {constB=0.45}
  n4 = TextureCoordinate()
  n5 = AppendVector(A=ScalarParameter() {defaultValue=2, parameterName=UTiling}, B=ScalarParameter() {defaultValue=1, parameterName=VTiling})
  n6 = TextureSample(UVs=Panner(Coordinate=Multiply(A=@n4, B=@n5), Speed=VectorParameter() {parameterName=ColorPanning, defaultValue=<rgba -0.1 0 0 1>}.RGB) {speedX=-0.1}) {samplerType=SAMPLERTYPE_LinearColor, texture=/VehicleCosmetics/DriftTrail/DriftTrail_Dragoncat/Textures/T_VCVFX_GroundTrail_D}
  n7 = Multiply(
   A=LinearInterpolate(A=@n6.RGB, B=Multiply(A=@n3, B=@n6.RGB), Alpha=ScalarParameter() {parameterName=GlowPann?})
   B=ScalarParameter() {defaultValue=1, parameterName=Brightness}
  )
  n8 = ScalarParameter() {defaultValue=1, parameterName=Constrast}
  n9 = ScalarParameter() {defaultValue=1, parameterName=Saturation}
  MaterialFunctionCall(Emissive=Desaturation(None=Power(Base=@n7, Exp=@n8), Fraction=OneMinus(None=@n9))) {materialFunction=&MF_FX_EmissiveAdjustments}
  MaterialExpressionVertexColor_1 = ParticleColor()
  n10 = MaterialFunctionCall(Float2=TextureCoordinate() {uTiling=2.6, vTiling=0.242}) {materialFunction=&BreakOutFloat2Components}
  n11 = Panner(Coordinate=TextureCoordinate() {uTiling=2.6, vTiling=0.02}) {speedX=-0.45}
  n12 = Multiply(
   A=ComponentMask(None=TextureSample(UVs=@n11) {texture=/DealershipSharedLibrary/FX/Textures/Noise/T_VCVFX_CloudDistort_A}.RGB) {g=true, r=true}
   B=ScalarParameter() {defaultValue=0.149, parameterName=MaxSpikeHeight}
  )
  n13 = Add(A=AppendVector(A=@n10.R, B=SmoothStep(Value=@n10.G) {constMax=0.3}), B=@n12)
  n14 = Subtract(A=@n10.G) {constB=0.242}
  n15 = MaterialFunctionCall(Float2=AppendVector(A=@n10.R, B=Abs(None=@n14))) {materialFunction=&BreakOutFloat2Components}
  n16 = Add(A=AppendVector(A=@n15.R, B=SmoothStep(Value=@n15.G) {constMax=0.3}), B=@n12)
  n17 = Add(
   A=TextureSample(UVs=Panner(Coordinate=@n13) {speedX=1}) {texture=/VehicleCosmetics/DriftTrail/DriftTrail_Dragoncat/Textures/T_VCVFX_DistortionWaves_A}.B
   B=TextureSample(UVs=Panner(Coordinate=@n16) {speedX=1}) {texture=&T_VCVFX_DistortionWaves_A}.B
  )
  n18 = TextureCoordinate() {uTiling=2.6, vTiling=0.3}
  n19 = TextureCoordinate()
  n20 = Multiply(A=OneMinus(None=ComponentMask(None=@n19) {g=true}), B=@n12)
  n21 = TextureSample(UVs=Panner(Coordinate=Add(A=@n18, B=@n20)) {speedX=-1}) {texture=&T_VCVFX_DistortionWaves_A}
  n22 = ScalarParameter() {defaultValue=1, parameterName=MirrorWaveT?}
  n23 = TextureCoordinate() {uTiling=45}
  n24 = Multiply(
   A=OneMinus(None=LinearInterpolate(A=@n17, B=@n21.B, Alpha=@n22))
   B=TextureSample(UVs=Panner(Coordinate=@n23) {speedX=-0.1}) {texture=/VehicleCosmetics/DriftTrail/DriftTrail_Dragoncat/Textures/T_VCVFX_Dots_A}.RGB
  )
  n25 = ScalarParameter() {defaultValue=1, parameterName=WavesT?}
  n26 = Constant2Vector() {r=2}
  n27 = ComponentMask(None=Multiply(A=TextureCoordinate(), B=Constant() {r=0.25})) {g=true, r=true}
  n28 = TextureSample(UVs=Panner(Coordinate=Multiply(A=@n26, B=@n27)) {speedX=-0.1, speedY=-0.4}) {texture=/VehicleCosmetics/DriftTrail/DriftTrail_Dragoncat/Textures/T_VCVFX_RandomSquare_A}
  n29 = Multiply(A=Panner(Coordinate=TextureCoordinate() {uTiling=64, vTiling=16}) {speedX=-0.025}, B=Constant() {r=1.045})
  n30 = ScalarParameter() {defaultValue=1, group=Mask, parameterName=Mask_Tile}
  n31 = ComponentMask(None=Frac(None=Multiply(A=@n29, B=@n30))) {r=true}
  n32 = ScalarParameter() {defaultValue=0.05, group=Mask, parameterName=Mask_Square_Size_V}
  n33 = ScalarParameter() {defaultValue=255, group=Mask, parameterName=Mask_Square_Sharpness}
  MaterialExpressionConstantClamp_3 = Saturate(None=Multiply(A=Subtract(A=@n31, B=@n32), B=@n33))
  n34 = OneMinus(None=@n31)
  MaterialExpressionConstantClamp_4 = Saturate(None=Multiply(A=Subtract(A=@n34, B=@n32), B=@n33))
  Multiply(
   A=@MaterialExpressionVertexColor_1.A
   B=Multiply(A=LinearInterpolate(A=@n28.RGB, B=@n24, Alpha=@n25), B=Multiply(A=@MaterialExpressionConstantClamp_3, B=@MaterialExpressionConstantClamp_4)) {constA=1}
  )
 ]
}
